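# Editor utility widget that launches the external VFX Browser from the editor through the fxbrowser_bridge Python module.
# On Construct, inline Python looks for launch_info.json in LOCALAPPDATA, and the first two captured log entries are shown in the widget.
# A substring test on that log either shows a not-ready message or reports ready and auto-opens the browser.
# The Restart button runs the same open command directly, without the check.
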EditorUtilityWidgetBlueprint /Game/NeverCook/FX/FXBrowser/EUW_VFXBrowserLauncher
graphs [EventGraph {nodes [
   n1 = "On Clicked (EditorUtilityButton_15)"() {is=title K2Node_ComponentBoundEvent}
   "Execute Python Command (Advanced)"!(
    execute=[@n1.then, @n7.then]
    PythonCommand:String="import fxbrowser_bridge; fxbrowser_bridge.open_fxbrowser()"
    ExecutionMode:EPythonCommandExecutionMode Enum=ExecuteFile
    FileExecutionScope:EPythonFileExecutionScope Enum=Private
   ) -> (then, CommandResult:String, LogOutput:Array of Python Log Output Entry Structures, ReturnValue:bool)
   Event Construct()
   > n2 = "Execute Python Command (Advanced)"(
    PythonCommand="import os, json; p = os.path.join(os.environ.get(\"LOCALAPPDATA\",\"\"), \"VFXBrowser\", \"UserData\", \"launch_info.json\"); print(\"[launch_info]\", \"FOUND\" if os.path.isfile(p) else \"NOT FOUND\", \"->\", p); print(open(p, encoding=\"utf-8\").read() if os.path.isfile(p) else \"(run FXBrowser from LaunchBox once to create it)\")"
    ExecutionMode=ExecuteFile
    FileExecutionScope=Private
   )
   n3 = Break Python Log Output Entry!(PythonLogOutputEntry:"Python Log Output Entry Structure (by ref)"="Get (a copy)"(Array:Array of Python Log Output Entry Structures=@n2.LogOutput, Dimension 1=1) -> (Output:Python Log Output Entry Structure)) -> (Type:EPythonLogOutputType Enum, Output:String) {is=title K2Node_BreakStruct}
   n4 = Break Python Log Output Entry(PythonLogOutputEntry="Get (a copy)"(Array=@n2.LogOutput))
   > Set Text(
    Text="To Text (String)"(InString=Append!(A=@n4.Output, B, C:String=@n3.Output) -> (ReturnValue))
    self=Get UEFN_TextBlock_C_0() -> (UEFN_TextBlock_C_0:Text Block Object Reference)
   )
   > n5 = Branch(Condition="integer >= integer"(A=FindSubstring(SearchIn=@n4.Output, Substring=NOT FOUND, bUseCase=true)))
   n6 = Get TextBlock_0() -> (TextBlock_0:Text Object Reference)
   > Set Text(Text="LOCTEXT(\"FXBrowser Not Ready, Please Start Once From LBox\")", self=@n6)
   n7 = Set Text(execute=@n5.else, Text="LOCTEXT(\"FXBrowser Ready, Auto starting....\")", self=@n6)
   Event PreConstruct()
   Event Tick()
   Get TextBlock_0()
  ]}]
variables []
widgets {
 info {rootWidgetClass=&CanvasPanel, widgetCount=5}
 widgets [
  CanvasPanel_54 {widgetClassPath=&CanvasPanel, widget=CanvasPanel_54}
  UEFN_TextBlock_C_0 {
   widgetClassPath=&UEFN_TextBlock
   bIsVariable=true
   parent=CanvasPanel_54
   widget=UEFN_TextBlock_C_0
   properties {
    text=Text Block
    font {
     bForceMonospaced=false
     bMaterialIsStencil=false
     fontMaterial=null
     fontObject=/LauncherSocial/BrutalType
     letterSpacing=0
     monospacedWidth=1
     size=12
     skewAmount=0
     typefaceFontName=Regular
     outlineSettings {
      bApplyOutlineToDropShadows=false
      bMiteredCorners=false
      bSeparateFillAlpha=false
      outlineMaterial=null
      outlineSize=0
      outlineColor=<rgba 0 0 0 1>
     }
    }
   }
   slot {
    class=CanvasPanelSlot
    properties {layoutData {alignment=<0 0>, anchors {maximum=<0 0>, minimum=<0 0>}, offsets=<ltrb 32 196 942.1772 433.7928>}}
   }
  }
  TextBlock_0 {
   widgetClassPath=&TextBlock
   bIsVariable=true
   parent=CanvasPanel_54
   widget=TextBlock_0
   properties {text=Text Block}
   slot {class=CanvasPanelSlot, properties {layoutData {alignment=<0 0>, anchors {maximum=<0 0>, minimum=<0 0>}, offsets=<ltrb 32 28 151 40>}}}
  }
  EditorUtilityButton_15 {
   widgetClassPath=&EditorUtilityButton
   parent=CanvasPanel_54
   widget=EditorUtilityButton_15
   slot {class=CanvasPanelSlot, properties {layoutData {alignment=<0 0>, anchors {maximum=<0 0>, minimum=<0 0>}, offsets=<ltrb 36 96 100 40>}}}
  }
  TextBlock_71 {
   widgetClassPath=&TextBlock
   parent=EditorUtilityButton_15
   widget=TextBlock_71
   properties {
    text=Restart
    font {
     bForceMonospaced=false
     bMaterialIsStencil=false
     fontMaterial=null
     fontObject=&Roboto
     letterSpacing=0
     monospacedWidth=1
     size=15
     skewAmount=0
     typefaceFontName=Bold
     outlineSettings {
      bApplyOutlineToDropShadows=false
      bMiteredCorners=false
      bSeparateFillAlpha=false
      outlineMaterial=null
      outlineSize=0
      outlineColor=<rgba 0 0 0 1>
     }
    }
   }
   slot {class=ButtonSlot}
  }
 ]
}
